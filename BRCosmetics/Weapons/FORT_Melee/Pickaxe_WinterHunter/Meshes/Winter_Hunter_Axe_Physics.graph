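# Physics asset for the Winter Hunter pickaxe cosmetic mesh, providing collision and joint setup for its belt bones.
# Two small capsules on dyn_belt and dyn_belt1 are each constrained to a larger angled capsule on dyn_master_belt1.
# Both joints share moderate swing and twist limits, letting the belt pieces move within a bounded cone around the master body.
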PhysicsAsset /BRCosmetics/Weapons/FORT_Melee/Pickaxe_WinterHunter/Meshes/Winter_Hunter_Axe_Physics
bodies {
 dyn_belt [{shapeType=Capsule, length=8, radius=2, rotation=<rot 90 0 0>}]
 dyn_belt1 [{shapeType=Capsule, length=8, radius=2, center=<0 0 0.001>, rotation=<rot 90 0 0>}]
 dyn_master_belt1 [{shapeType=Capsule, length=30, radius=12, center=<-3.1692 0 -1.0191>, rotation=<rot 17.8251 0 0>}]
}
constraints [
 {bone1Name=dyn_belt, bone2Name=dyn_master_belt1, swing1LimitDegrees=45, swing2LimitDegrees=45, twistLimitDegrees=45}
 {bone1Name=dyn_belt1, bone2Name=dyn_master_belt1, swing1LimitDegrees=45, swing2LimitDegrees=45, twistLimitDegrees=45}
]
